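# Niagara system for the PS Plus variant of the Lava vehicle boost cosmetic, drawing a trail behind the car's boost ports.
# Emitters include a flipbook sprite main trail, ribbons, splash sprites and glow, placed per port on filtered skeletal mesh sockets.
# Curves drive green-yellow color, size and fade, while the main trail adds drag, gravity and kill-on-collision.
# Game-mode masks and preview modules adjust the effect; several painted-variant overrides are disabled.
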NiagaraSystem /VehicleCosmetics/Boosts/Boost_Lava_Psplus/FX/NS_Boost_LavaPsplus
dependencies {
 usedDataInterfaces [
  &NiagaraDataInterfaceCurve
  &NiagaraDataInterfaceSpriteRendererInfo
  &NiagaraDataInterfaceColorCurve
  &NiagaraDataInterfaceArrayFloat
  &NiagaraDataInterfaceSkeletalMesh
  &NiagaraDataInterfaceArrayFloat3
 ]
 usedDynamicInputs [&SelectVectorFromArray]
 usedModules [
  &self:SystemSpawnScript.SystemScriptSource.SystemScriptGraph.NiagaraNodeAssignment_0.SetVariables_#1
  &self:SystemSpawnScript.SystemScriptSource.SystemScriptGraph.NiagaraNodeAssignment_1.SetVariables_#2
  &SystemState
  &self:SystemSpawnScript.SystemScriptSource.SystemScriptGraph.NiagaraNodeAssignment_2.SetVariables_#3
 ]
}
emitters {
 NE_CS_MainTrail002 {
  emitterUpdateScript [
   &EmitterState()
   &VC_FilteredSocketCount()
   &SpawnRate(
    SpawnRate=&Multiply_Float(
     A=600
     B=/VehicleCosmetics/FX/Niagara/Modules/VC_Mask_Floats(Coro Value:NiagaraFloat=3, FrontEnd Value:NiagaraFloat=0.5, FN Value:NiagaraFloat=0.5)
    )
    Use Spawn Probability=false
   )
  ]
  particleSpawnScript [
   /SetVariables(Particles.VisTag:NiagaraInt32=&SpawnGroup()) {isSetParameters=true, module=SetVariables_#4}
   &VC_PaintedVariantsOverride(
    R2 Black:LinearColor=<rgba 0.16 0.16 0.16 1>
    R2 DefaultColor:LinearColor=<rgba 0.9348 0.5504 0.9896 1>
    Use R1 Painted Overrides=true
    Use R2 Painted Overrides=true
   ) {enabled=false}
   &InitializeParticle(
    Lifetime Max=0.5
    Lifetime Min=0.35
    Lifetime Mode=::NewEnumerator1
    Sprite Rotation Mode=::NewEnumerator2
    Sprite Size Mode=::NewEnumerator1
    Uniform Sprite Size Max=160
    Uniform Sprite Size Min=140
    UsePositionOffset=false
    Write Material Random=true
   )
   &PartitionParticles(Number Of Partitions=$Emitter.PortCount)
   &SkeletalMeshLocation(
    Bone / Socket Index=$Particles.PartitionParticles.Partition
    Bone / Socket Sampling Mode=::NewEnumerator5
    MaskSpawnGroup=false
    Mesh Sampling Type=::NewEnumerator6
    Sampled Position Offset=&AddVectorToPosition(Position=<0 0 0>, Vector=$User.DTCarOffset)
    Skeletal Mesh=$System.SkeletalMesh
   )
   &AddVelocity(Velocity=<-1500 0 0>, Velocity Speed Scale=1.13)
   /VehicleCosmetics/FX/VC_FXParents/Modules/DLR_ProductPreviews(GameModeID:NiagaraInt32=$System.GameModeID, bPreviewVehicleSpeed:NiagaraBool=$System.bPreviewVehicleSpeed)
  ]
  particleUpdateScript [
   &ParticleState()
   /SetVariables(
    Particles.Glow_Color:LinearColor=&ColorFromCurve(ColorCurve=DataInterface() {propertyValues {
       AlphaCurve {keys [<at 0 0>,<at 0.0459 1 RCIM_Cubic>,<at 0.6733 0 RCIM_Cubic>]}
       BlueCurve {keys [<at 0 0 RCIM_Cubic>]}
       GreenCurve {keys [<at 0 0.7374 RCIM_Cubic>]}
       RedCurve {keys [<at 0 2 RCIM_Cubic>]}
      }})
    Particles.Glow_Size:Vector2f=&RandomRangeVector2D(Minimum=<100 100>, Maximum=<200 200>)
   ) {isSetParameters=true, module=SetVariables_#5}
   &VC_PaintedVariantsOverride(
    Black=&ColorFromCurve(ColorCurve=DataInterface() {propertyValues {
       AlphaCurve {keys [<at 0.0801 0 RCIM_Cubic>,<at 0.119 1 RCIM_Cubic>]}
       BlueCurve {keys [<at 0.1205 0.01 RCIM_Cubic>]}
       GreenCurve {keys [<at 0.1205 0.01 RCIM_Cubic>]}
       RedCurve {keys [<at 0.1205 0.01 RCIM_Cubic>]}
      }})
    BurntSienna=&ColorFromCurve(ColorCurve=DataInterface() {propertyValues {
       AlphaCurve {keys [<at 0.0646 0>,<at 0.1641 1 RCIM_Cubic>,<at 1 0 RCIM_Cubic>]}
       BlueCurve {keys [<at 0 0.0075 RCIM_Cubic>,<at 0.6548 0.002 RCIM_Cubic>]}
       GreenCurve {keys [<at 0 0.0382 RCIM_Cubic>,<at 0.6548 0.0367 RCIM_Cubic>]}
       RedCurve {keys [<at 0 0.2346 RCIM_Cubic>,<at 0.6548 0.0305 RCIM_Cubic>]}
      }})
    Cobalt=&ColorFromCurve(ColorCurve=DataInterface() {propertyValues {
       AlphaCurve {keys [<at 0.0614 0 RCIM_Cubic>,<at 0.1719 1 RCIM_Cubic>,<at 0.6292 0.0028>]}
       BlueCurve {keys [<at 0 1 RCIM_Cubic>,<at 0.8633 0 RCIM_Cubic>]}
       GreenCurve {keys [<at 0 0 RCIM_Cubic>,<at 0.8633 0 RCIM_Cubic>]}
       RedCurve {keys [<at 0 0.0003 RCIM_Cubic>,<at 0.8633 0 RCIM_Cubic>]}
      }})
    Crimson=&ColorFromCurve(ColorCurve=DataInterface() {propertyValues {
       AlphaCurve {keys [<at 0.0381 0>,<at 0.1921 1 RCIM_Cubic>,<at 0.6836 1>,<at 1 0 RCIM_Cubic>]}
       BlueCurve {keys [<at 0 0.0033 RCIM_Cubic>,<at 0.1921 0.0125>,<at 0.4736 0.0234 RCIM_Cubic>]}
       GreenCurve {keys [<at 0 0.0033 RCIM_Cubic>,<at 0.1921 0.0257>,<at 0.4736 0.3926 RCIM_Cubic>]}
       RedCurve {keys [<at 0 0.1875 RCIM_Cubic>,<at 0.1921 0.6771>,<at 0.4736 0.4948 RCIM_Cubic>]}
      }})
    DefaultColor=&ColorFromCurve(ColorCurve=DataInterface() {propertyValues {
       AlphaCurve {keys [<at 0 0 RCIM_Cubic>,<at 0.1719 1 RCIM_Cubic>]}
       BlueCurve {keys [<at 0.1753 0.5156 RCIM_Cubic>,<at 0.6077 1>]}
       GreenCurve {keys [<at 0.1753 0.2305 RCIM_Cubic>,<at 0.6077 0.3073>]}
       RedCurve {keys [<at 0.1753 0.2585 RCIM_Cubic>,<at 0.6077 0.8652>]}
      }})
    ForestGreen=&ColorFromCurve(ColorCurve=DataInterface() {propertyValues {
       AlphaCurve {keys [<at 0.0381 0>,<at 0.1921 1 RCIM_Cubic>,<at 0.6836 1>,<at 1 0 RCIM_Cubic>]}
       BlueCurve {keys [<at 0 0.0856 RCIM_Cubic>,<at 0.259 0.0804>,<at 0.4301 0.5573 RCIM_Cubic>]}
       GreenCurve {keys [<at 0 0.5826 RCIM_Cubic>,<at 0.259 0.5469>,<at 0.4301 0.5279 RCIM_Cubic>]}
       RedCurve {keys [<at 0 0 RCIM_Cubic>,<at 0.259 0>,<at 0.4301 0 RCIM_Cubic>]}
      }})
    Grey=&ColorFromCurve(ColorCurve=DataInterface() {propertyValues {
       AlphaCurve {keys [<at 0.0303 0 RCIM_Cubic>,<at 0.1439 1 RCIM_Cubic>]}
       BlueCurve {keys [<at 0 0.0999 RCIM_Cubic>,<at 0.2745 0.099 RCIM_Cubic>]}
       GreenCurve {keys [<at 0 0.0999 RCIM_Cubic>,<at 0.2745 0.099 RCIM_Cubic>]}
       RedCurve {keys [<at 0 0.0999 RCIM_Cubic>,<at 0.2745 0.099 RCIM_Cubic>]}
      }})
    Lime=&ColorFromCurve(ColorCurve=DataInterface() {propertyValues {
       AlphaCurve {keys [
         <at 0.0506 0 RCIM_Cubic>
         {arriveTangent=2.6348, interpMode=RCIM_Cubic, leaveTangent=2.6348, time=0.1859, value=1}
         <at 0.4301 1>
         <at 0.6619 0>
        ]}
       BlueCurve {keys [<at 0.1828 0 RCIM_Cubic>,<at 0.3834 0.061 RCIM_Cubic>]}
       GreenCurve {keys [<at 0.1828 0.7917 RCIM_Cubic>,<at 0.3834 0.1476 RCIM_Cubic>]}
       RedCurve {keys [<at 0.1828 0.1696 RCIM_Cubic>,<at 0.3834 0.0081 RCIM_Cubic>]}
      }})
    Orange=&ColorFromCurve(ColorCurve=DataInterface() {propertyValues {
       AlphaCurve {keys [<at 0 0 RCIM_Cubic>, {arriveTangent=2.6348, interpMode=RCIM_Cubic, time=0.1859, value=1}, <at 0.4301 1>, <at 0.6619 0>]}
       BlueCurve {keys [<at 0.1828 0.0008 RCIM_Cubic>,<at 0.3834 0 RCIM_Cubic>]}
       GreenCurve {keys [<at 0.1828 0.2332 RCIM_Cubic>,<at 0.3834 0.3646 RCIM_Cubic>]}
       RedCurve {keys [<at 0.1828 0.6146 RCIM_Cubic>,<at 0.3834 0.2732 RCIM_Cubic>]}
      }})
    Pink=&ColorFromCurve(ColorCurve=DataInterface() {propertyValues {
       AlphaCurve {keys [
         <at 0 0 RCIM_Cubic>
         {arriveTangent=2.3251, interpMode=RCIM_Cubic, leaveTangent=2.3251, time=0.1859, value=0.8}
         <at 0.4301 1>
         <at 0.6034 0>
        ]}
       BlueCurve {keys [<at 0.1828 0.147 RCIM_Cubic>,<at 0.5031 0.1628 RCIM_Cubic>]}
       GreenCurve {keys [<at 0.1828 0.0821 RCIM_Cubic>,<at 0.5031 0.1997 RCIM_Cubic>]}
       RedCurve {keys [<at 0.1828 0.1771 RCIM_Cubic>,<at 0.5031 0.3906 RCIM_Cubic>]}
      }})
    Purple=&ColorFromCurve(ColorCurve=DataInterface() {propertyValues {
       AlphaCurve {keys [
         <at 0 0 RCIM_Cubic>
         {arriveTangent=2.3251, interpMode=RCIM_Cubic, leaveTangent=2.3251, time=0.1099, value=0.8}
         <at 0.4301 1>
         <at 0.6034 0>
        ]}
       BlueCurve {keys [<at 0.1828 0.1771 RCIM_Cubic>,<at 0.5031 0.5648 RCIM_Cubic>]}
       GreenCurve {keys [<at 0.1828 0.0338 RCIM_Cubic>,<at 0.5031 0.0011 RCIM_Cubic>]}
       RedCurve {keys [<at 0.1828 0.138 RCIM_Cubic>,<at 0.5031 0.5885 RCIM_Cubic>]}
      }})
    R1 Black:LinearColor=&ColorFromCurve(ColorCurve=DataInterface() {propertyValues {
       AlphaCurve {keys [<at 0.1097 0 RCIM_Cubic>,<at 0.2745 1 RCIM_Cubic>,<at 0.6836 0>]}
       BlueCurve {keys [<at 0.1091 0.01 RCIM_Cubic>,<at 0.2892 0.1771 RCIM_Cubic>]}
       GreenCurve {keys [<at 0.1091 0.01 RCIM_Cubic>,<at 0.2892 0.1771 RCIM_Cubic>]}
       RedCurve {keys [<at 0.1091 0.01 RCIM_Cubic>,<at 0.2892 0.1771 RCIM_Cubic>]}
      }})
    R1 BurntSienna:LinearColor=&ColorFromCurve(ColorCurve=DataInterface() {propertyValues {
       AlphaCurve {keys [<at 0 0>,<at 0.469 1 RCIM_Cubic>,<at 0.6525 0 RCIM_Cubic>]}
       BlueCurve {keys [<at 0 0.0178 RCIM_Cubic>,<at 0.6548 0.0108 RCIM_Cubic>]}
       GreenCurve {keys [<at 0 0.0908 RCIM_Cubic>,<at 0.6548 0.6927 RCIM_Cubic>]}
       RedCurve {keys [<at 0 0.5573 RCIM_Cubic>,<at 0.6548 0.57 RCIM_Cubic>]}
      }})
    R1 Cobalt:LinearColor=&ColorFromCurve(ColorCurve=DataInterface() {propertyValues {
       AlphaCurve {keys [<at 0.1174 0 RCIM_Cubic>,<at 0.2528 1 RCIM_Cubic>,<at 0.6105 0>]}
       BlueCurve {keys [<at 0.1753 1 RCIM_Cubic>,<at 0.6077 3>]}
       GreenCurve {keys [<at 0.1753 0.4069 RCIM_Cubic>,<at 0.6077 0>]}
       RedCurve {keys [<at 0.1753 0.3646 RCIM_Cubic>,<at 0.6077 2.5078>]}
      }})
    R1 Crimson:LinearColor=&ColorFromCurve(ColorCurve=DataInterface() {propertyValues {
       AlphaCurve {keys [
         <at 0.147 0 RCIM_Cubic>
         {arriveTangent=3.9441, interpMode=RCIM_Cubic, leaveTangent=3.9441, time=0.2808, value=1}
         <at 0.4005 1>
         <at 0.6074 0>
        ]}
       BlueCurve {keys [<at 0.1753 0.1286 RCIM_Cubic>,<at 0.6077 0>]}
       GreenCurve {keys [<at 0.1753 0.1763 RCIM_Cubic>,<at 0.6077 1.4989>]}
       RedCurve {keys [<at 0.1753 1 RCIM_Cubic>,<at 0.6077 1.5>]}
      }})
    R1 DefaultColor=&ColorFromCurve(ColorCurve=DataInterface() {propertyValues {
       AlphaCurve {keys [<at 0.0817 0 RCIM_Cubic>,<at 0.2357 1 RCIM_Cubic>,<at 0.6821 0>]}
       BlueCurve {keys [<at 0.1753 1 RCIM_Cubic>,<at 0.6077 2>]}
       GreenCurve {keys [<at 0.1753 0.8125 RCIM_Cubic>,<at 0.6077 0.6875>]}
       RedCurve {keys [<at 0.1753 0.831 RCIM_Cubic>,<at 0.6077 1.7446>]}
      }})
    R1 ForestGreen:LinearColor=&ColorFromCurve(ColorCurve=DataInterface() {propertyValues {
       AlphaCurve {keys [<at 0.0739 0 RCIM_Cubic>, {arriveTangent=3.9441, interpMode=RCIM_Cubic, time=0.2808, value=1}, <at 0.4005 1>, <at 0.6074 0>]}
       BlueCurve {keys [{arriveTangent=3.2911, interpMode=RCIM_Cubic, time=0.1753, value=0}, <at 0.6077 2>]}
       GreenCurve {keys [{arriveTangent=2.7598, interpMode=RCIM_Cubic, time=0.1753, value=0.8594}, <at 0.6077 2>]}
       RedCurve {keys [{arriveTangent=-0.1276, interpMode=RCIM_Cubic, time=0.1753, value=0.2064}, <at 0.6077 0>]}
      }})
    R1 Grey:LinearColor=&ColorFromCurve(ColorCurve=DataInterface() {propertyValues {
       AlphaCurve {keys [<at 0.1143 0>,<at 0.3088 1 RCIM_Cubic>,<at 0.6043 0 RCIM_Cubic>]}
       BlueCurve {keys [<at 0.3072 0.6562 RCIM_Cubic>]}
       GreenCurve {keys [<at 0.3072 0.6562 RCIM_Cubic>]}
       RedCurve {keys [<at 0.3072 0.6562 RCIM_Cubic>]}
      }})
    R1 Lime:LinearColor=&ColorFromCurve(ColorCurve=DataInterface() {propertyValues {
       AlphaCurve {keys [<at 0.0739 0 RCIM_Cubic>, {arriveTangent=3.9441, interpMode=RCIM_Cubic, time=0.2808, value=1}, <at 0.4005 1>, <at 0.6074 0>]}
       BlueCurve {keys [<at 0.091 0 RCIM_Cubic>,<at 0.6077 1.3543>]}
       GreenCurve {keys [<at 0.091 2 RCIM_Cubic>,<at 0.6077 2>]}
       RedCurve {keys [<at 0.091 1.1387 RCIM_Cubic>,<at 0.6077 0>]}
      }})
    R1 Orange:LinearColor=&ColorFromCurve(ColorCurve=DataInterface() {propertyValues {
       AlphaCurve {keys [<at 0.0739 0 RCIM_Cubic>, {arriveTangent=3.9441, interpMode=RCIM_Cubic, time=0.2808, value=1}, <at 0.4005 1>, <at 0.6074 0>]}
       BlueCurve {keys [<at 0.091 0 RCIM_Cubic>,<at 0.6077 0.2815>]}
       GreenCurve {keys [<at 0.091 0.7377 RCIM_Cubic>,<at 0.6077 1>]}
       RedCurve {keys [<at 0.091 2 RCIM_Cubic>,<at 0.6077 0.7462>]}
      }})
    R1 Pink:LinearColor=&ColorFromCurve(ColorCurve=DataInterface() {propertyValues {
       AlphaCurve {keys [<at 0.0739 0 RCIM_Cubic>, {arriveTangent=3.9441, interpMode=RCIM_Cubic, time=0.2808, value=1}, <at 0.4005 1>, <at 0.6074 0>]}
       BlueCurve {keys [<at 0.091 1.7582 RCIM_Cubic>,<at 0.3276 0.4062>]}
       GreenCurve {keys [<at 0.091 0.1064 RCIM_Cubic>,<at 0.3276 0.8298>]}
       RedCurve {keys [<at 0.091 2 RCIM_Cubic>,<at 0.3276 2>]}
      }})
    R1 Purple:LinearColor=&ColorFromCurve(ColorCurve=DataInterface() {propertyValues {
       AlphaCurve {keys [<at 0.1131 0 RCIM_Cubic>, {arriveTangent=3.9441, interpMode=RCIM_Cubic, time=0.2466, value=1}, <at 0.4005 1>, <at 0.6074 0>]}
       BlueCurve {keys [<at 0.2037 5 RCIM_Cubic>,<at 0.3276 0.709>]}
       GreenCurve {keys [<at 0.2037 0 RCIM_Cubic>,<at 0.3276 0.1062>]}
       RedCurve {keys [<at 0.2037 0.9897 RCIM_Cubic>,<at 0.3276 1>]}
      }})
    R1 Saffron:LinearColor=&ColorFromCurve(ColorCurve=DataInterface() {propertyValues {
       AlphaCurve {keys [<at 0.1131 0 RCIM_Cubic>, {arriveTangent=3.9441, interpMode=RCIM_Cubic, time=0.2466, value=1}, <at 0.4005 1>, <at 0.6074 0>]}
       BlueCurve {keys [<at 0.1542 0.875 RCIM_Cubic>,<at 0.3276 0.533>]}
       GreenCurve {keys [<at 0.1542 2.7603 RCIM_Cubic>,<at 0.3276 2>]}
       RedCurve {keys [<at 0.1542 3 RCIM_Cubic>,<at 0.3276 0.9975>]}
      }})
    R1 SkyBlue:LinearColor=&ColorFromCurve(ColorCurve=DataInterface() {propertyValues {
       AlphaCurve {keys [<at 0.1131 0 RCIM_Cubic>, {arriveTangent=3.9441, interpMode=RCIM_Cubic, time=0.2466, value=1}, <at 0.4005 1>, <at 0.6074 0>]}
       BlueCurve {keys [<at 0.1542 2 RCIM_Cubic>,<at 0.3276 1>]}
       GreenCurve {keys [<at 0.1542 1.715 RCIM_Cubic>,<at 0.3276 0.3668>]}
       RedCurve {keys [<at 0.1542 0.0312 RCIM_Cubic>,<at 0.3276 0.519>]}
      }})
    R1 TitaniumWhite:LinearColor=&ColorFromCurve(ColorCurve=DataInterface() {propertyValues {
       AlphaCurve {keys [<at 0.1131 0 RCIM_Cubic>, {arriveTangent=3.9441, interpMode=RCIM_Cubic, time=0.2466, value=1}, <at 0.4005 1>, <at 0.6074 0>]}
       BlueCurve {keys [<at 0.1542 2 RCIM_Cubic>,<at 0.3276 2>]}
       GreenCurve {keys [<at 0.1542 2 RCIM_Cubic>,<at 0.3276 2>]}
       RedCurve {keys [<at 0.1542 2 RCIM_Cubic>,<at 0.3276 2>]}
      }})
    R2 Black:LinearColor=&ColorFromCurve(ColorCurve=DataInterface() {propertyValues {
       AlphaCurve {keys [<at 0.0677 0 RCIM_Cubic>,<at 0.2668 0.8 RCIM_Cubic>,<at 1 0>]}
       BlueCurve {keys [<at 0.1091 0.001 RCIM_Cubic>,<at 0.2892 0.02 RCIM_Cubic>]}
       GreenCurve {keys [<at 0.1091 0.001 RCIM_Cubic>,<at 0.2892 0.02 RCIM_Cubic>]}
       RedCurve {keys [<at 0.1091 0.001 RCIM_Cubic>,<at 0.2892 0.02 RCIM_Cubic>]}
      }})
    R2 BurntSienna:LinearColor=&ColorFromCurve(ColorCurve=DataInterface() {propertyValues {
       AlphaCurve {keys [<at 0.1765 0 RCIM_Cubic>,<at 0.4052 0.6 RCIM_Cubic>]}
       BlueCurve {keys [<at 0.2419 0.0062 RCIM_Cubic>,<at 0.3865 0.0191 RCIM_Cubic>]}
       GreenCurve {keys [<at 0.2419 0.0287 RCIM_Cubic>,<at 0.3865 0.0573 RCIM_Cubic>]}
       RedCurve {keys [<at 0.2419 0.1771 RCIM_Cubic>,<at 0.3865 0.0504 RCIM_Cubic>]}
      }})
    R2 Cobalt:LinearColor=&ColorFromCurve(ColorCurve=DataInterface() {propertyValues {
       AlphaCurve {keys [<at 0.0817 0 RCIM_Cubic>,<at 0.4239 0.6 RCIM_Cubic>,<at 0.8485 0>]}
       BlueCurve {keys [<at 0.1753 0.3281 RCIM_Cubic>,<at 0.4954 0.224>]}
       GreenCurve {keys [<at 0.1753 0.0218 RCIM_Cubic>,<at 0.4954 0>]}
       RedCurve {keys [<at 0.1753 0 RCIM_Cubic>,<at 0.4954 0.1872>]}
      }})
    R2 Crimson:LinearColor=&ColorFromCurve(ColorCurve=DataInterface() {propertyValues {
       AlphaCurve {keys [<at 0.0817 0 RCIM_Cubic>,<at 0.3492 0.8 RCIM_Cubic>,<at 0.8485 0>]}
       BlueCurve {keys [<at 0.1753 0 RCIM_Cubic>,<at 0.4954 0.0137>]}
       GreenCurve {keys [<at 0.1753 0.0039 RCIM_Cubic>,<at 0.4954 0.0729>]}
       RedCurve {keys [<at 0.1753 0.2083 RCIM_Cubic>,<at 0.4954 0.0566>]}
      }})
    R2 DefaultColor:LinearColor=&ColorFromCurve(ColorCurve=DataInterface() {propertyValues {
       AlphaCurve {keys [<at 0.1672 0 RCIM_Cubic>,<at 0.3803 0.7 RCIM_Cubic>,<at 0.7334 0>]}
       BlueCurve {keys [<at 0.1753 0.1198 RCIM_Cubic>,<at 0.3508 0.1771>]}
       GreenCurve {keys [<at 0.1753 0.0381 RCIM_Cubic>,<at 0.3508 0.0839>]}
       RedCurve {keys [<at 0.1753 0.0461 RCIM_Cubic>,<at 0.3508 0.159>]}
      }})
    R2 ForestGreen:LinearColor=&ColorFromCurve(ColorCurve=DataInterface() {propertyValues {
       AlphaCurve {keys [
         <at 0.049 0>
         {arriveTangent=1.2606, interpMode=RCIM_Cubic, leaveTangent=1.2606, time=0.4161, value=0.8}
         <at 0.6836 0.8>
         <at 1 0 RCIM_Cubic>
        ]}
       BlueCurve {keys [<at 0 0.0334>,<at 0.4736 0.0922 RCIM_Cubic>]}
       GreenCurve {keys [<at 0 0.1823>,<at 0.4736 0.1198 RCIM_Cubic>]}
       RedCurve {keys [<at 0 0>,<at 0.4736 0 RCIM_Cubic>]}
      }})
    R2 Grey:LinearColor=&ColorFromCurve(ColorCurve=DataInterface() {propertyValues {
       AlphaCurve {keys [
         <at 0.147 0 RCIM_Cubic>
         {arriveTangent=3.1553, interpMode=RCIM_Cubic, leaveTangent=3.1553, time=0.2808, value=0.8}
         <at 0.4005 0.8>
         <at 0.6074 0>
        ]}
       BlueCurve {keys [<at 0.0708 0.2083 RCIM_Cubic>,<at 0.3508 0.0833>]}
       GreenCurve {keys [<at 0.0708 0.2083 RCIM_Cubic>,<at 0.3508 0.0833>]}
       RedCurve {keys [<at 0.0708 0.2083 RCIM_Cubic>,<at 0.3508 0.0833>]}
      }})
    R2 Lime:LinearColor=&ColorFromCurve(ColorCurve=DataInterface() {propertyValues {
       AlphaCurve {keys [
         <at 0.0412 0 RCIM_Cubic>
         {arriveTangent=1.6698, interpMode=RCIM_Cubic, leaveTangent=1.6698, time=0.2808, value=0.6}
         <at 0.4005 0.6>
         <at 0.6074 0>
        ]}
       BlueCurve {keys [<at 0.0941 0.0002 RCIM_Cubic>,<at 0.3803 0.1206>]}
       GreenCurve {keys [<at 0.0941 0.3438 RCIM_Cubic>,<at 0.3803 0.1927>]}
       RedCurve {keys [<at 0.0941 0.2305 RCIM_Cubic>,<at 0.3803 0.0384>]}
      }})
    R2 Orange:LinearColor=&ColorFromCurve(ColorCurve=DataInterface() {propertyValues {
       AlphaCurve {keys [
         <at 0.0412 0 RCIM_Cubic>
         {arriveTangent=1.6698, interpMode=RCIM_Cubic, leaveTangent=1.6698, time=0.1552, value=0.5}
         <at 0.4005 0.6>
         <at 0.6074 0>
        ]}
       BlueCurve {keys [<at 0.1583 0 RCIM_Cubic>,<at 0.3803 0>]}
       GreenCurve {keys [<at 0.1583 0.2274 RCIM_Cubic>,<at 0.3803 0.1458>]}
       RedCurve {keys [<at 0.1583 0.5469 RCIM_Cubic>,<at 0.3803 0.1076>]}
      }})
    R2 Pink:LinearColor=&ColorFromCurve(ColorCurve=DataInterface() {propertyValues {
       AlphaCurve {keys [
         <at 0.0412 0 RCIM_Cubic>
         {arriveTangent=1.6698, interpMode=RCIM_Cubic, leaveTangent=1.6698, time=0.1552, value=0.5}
         <at 0.4005 0.6>
         <at 0.6074 0>
        ]}
       BlueCurve {keys [<at 0.1583 0.2922 RCIM_Cubic>,<at 0.3803 0.0526>]}
       GreenCurve {keys [<at 0.1583 0.0423 RCIM_Cubic>,<at 0.3803 0.0787>]}
       RedCurve {keys [<at 0.1583 0.4271 RCIM_Cubic>,<at 0.3803 0.1094>]}
      }})
    R2 Purple:LinearColor=&ColorFromCurve(ColorCurve=DataInterface() {propertyValues {
       AlphaCurve {keys [
         <at 0.0412 0 RCIM_Cubic>
         {arriveTangent=1.6698, interpMode=RCIM_Cubic, leaveTangent=1.6698, time=0.1552, value=0.5}
         <at 0.4005 0.6>
         <at 0.6074 0>
        ]}
       BlueCurve {keys [<at 0.1583 1 RCIM_Cubic>,<at 0.3803 0.08>]}
       GreenCurve {keys [<at 0.1583 0 RCIM_Cubic>,<at 0.3803 0.0166>]}
       RedCurve {keys [<at 0.1583 0.287 RCIM_Cubic>,<at 0.3803 0.1094>]}
      }})
    R2 Saffron:LinearColor=&ColorFromCurve(ColorCurve=DataInterface() {propertyValues {
       AlphaCurve {keys [
         <at 0.0412 0 RCIM_Cubic>
         {arriveTangent=1.6698, interpMode=RCIM_Cubic, leaveTangent=1.6698, time=0.1552, value=0.5}
         <at 0.4005 0.6>
         <at 0.6074 0>
        ]}
       BlueCurve {keys [<at 0.1583 0 RCIM_Cubic>,<at 0.3803 0.0873>]}
       GreenCurve {keys [<at 0.1583 0.8117 RCIM_Cubic>,<at 0.3803 0.2396>]}
       RedCurve {keys [<at 0.1583 0.875 RCIM_Cubic>,<at 0.3803 0.1074>]}
      }})
    R2 SkyBlue:LinearColor=&ColorFromCurve(ColorCurve=DataInterface() {propertyValues {
       AlphaCurve {keys [
         <at 0.0412 0 RCIM_Cubic>
         {arriveTangent=1.6698, interpMode=RCIM_Cubic, leaveTangent=1.6698, time=0.1552, value=0.5}
         <at 0.4005 0.6>
         <at 0.6074 0>
        ]}
       BlueCurve {keys [<at 0.1583 0.4167 RCIM_Cubic>,<at 0.3803 0.125>]}
       GreenCurve {keys [<at 0.1583 0.2756 RCIM_Cubic>,<at 0.3803 0.041>]}
       RedCurve {keys [<at 0.1583 0.0434 RCIM_Cubic>,<at 0.3803 0.0705>]}
      }})
    R2 TitaniumWhite:LinearColor=&ColorFromCurve(ColorCurve=DataInterface() {propertyValues {
       AlphaCurve {keys [
         <at 0.0412 0 RCIM_Cubic>
         {arriveTangent=1.6698, interpMode=RCIM_Cubic, leaveTangent=1.6698, time=0.1552, value=0.5}
         <at 0.4005 0.6>
         <at 0.6074 0>
        ]}
       BlueCurve {keys [<at 0.1583 0.6667 RCIM_Cubic>,<at 0.3803 0.2031>]}
       GreenCurve {keys [<at 0.1583 0.6667 RCIM_Cubic>,<at 0.3803 0.2031>]}
       RedCurve {keys [<at 0.1583 0.6667 RCIM_Cubic>,<at 0.3803 0.2031>]}
      }})
    Saffron=&ColorFromCurve(ColorCurve=DataInterface() {propertyValues {
       AlphaCurve {keys [<at 0.0318 0 RCIM_Cubic>, {arriveTangent=2.3251, interpMode=RCIM_Cubic, time=0.1803, value=0.8}, <at 0.4301 1>, <at 0.6034 0>]}
       BlueCurve {keys [<at 0.1258 0 RCIM_Cubic>,<at 0.3258 0 RCIM_Cubic>]}
       GreenCurve {keys [<at 0.1258 0.7424 RCIM_Cubic>,<at 0.3258 0.4219 RCIM_Cubic>]}
       RedCurve {keys [<at 0.1258 0.875 RCIM_Cubic>,<at 0.3258 0.2191 RCIM_Cubic>]}
      }})
    SkyBlue=&ColorFromCurve(ColorCurve=DataInterface() {propertyValues {
       AlphaCurve {keys [<at 0.0437 0 RCIM_Cubic>, {arriveTangent=2.3251, interpMode=RCIM_Cubic, time=0.1654, value=0.8}, <at 0.4301 1>, <at 0.6034 0>]}
       BlueCurve {keys [<at 0.1022 0.4792 RCIM_Cubic>,<at 0.3602 0.4323 RCIM_Cubic>]}
       GreenCurve {keys [<at 0.1022 0.4551 RCIM_Cubic>,<at 0.3602 0.1302 RCIM_Cubic>]}
       RedCurve {keys [<at 0.1022 0.0393 RCIM_Cubic>,<at 0.3602 0.2357 RCIM_Cubic>]}
      }})
    TitaniumWhite=&ColorFromCurve(ColorCurve=DataInterface() {propertyValues {
       AlphaCurve {keys [<at 0.0437 0 RCIM_Cubic>, {arriveTangent=2.3251, interpMode=RCIM_Cubic, time=0.1654, value=0.8}, <at 0.4301 1>, <at 0.6034 0>]}
       BlueCurve {keys [<at 0.1022 1 RCIM_Cubic>,<at 0.3602 0.4323 RCIM_Cubic>]}
       GreenCurve {keys [<at 0.1022 1 RCIM_Cubic>,<at 0.3602 0.4323 RCIM_Cubic>]}
       RedCurve {keys [<at 0.1022 1 RCIM_Cubic>,<at 0.3602 0.4323 RCIM_Cubic>]}
      }})
    Use R1 Painted Overrides=true
    Use R2 Painted Overrides=true
   ) {enabled=false, module=VC_PaintedVariantsOverride001}
   &ScaleSpriteSize(
    Uniform Curve Scale=&RandomRangeFloat(Minimum=0.8, Maximum=0.9)
    Uniform Curve Sprite Scale=DataInterface() {propertyValues {Curve {keys [
        {arriveTangent=0.8007, interpMode=RCIM_Cubic, leaveTangent=0.8007, tangentMode=RCTM_User, time=0, value=0.7}
        {arriveTangent=-0.0118, interpMode=RCIM_Cubic, leaveTangent=-0.0118, tangentMode=RCTM_User, time=1, value=1}
       ]}}}
   )
   &Drag(
    Drag=&FloatFromCurve(
     FloatCurve=DataInterface() {propertyValues {Curve {
        defaultValue=1
        keys [
         {arriveTangent=0.0291, interpMode=RCIM_Cubic, leaveTangent=0.003, tangentMode=RCTM_Break, time=0, value=0}
         {arriveTangent=4.1537, interpMode=RCIM_Cubic, leaveTangent=4.1537, tangentMode=RCTM_User, time=1, value=2}
        ]
       }}}
     Scale Curve=1.5
    )
    Use Linear Drag=true
    Use Rotational Drag=true
   )
   &ScaleColor(
    Scale Alpha=&FloatFromCurve(
     FloatCurve=DataInterface() {propertyValues {Curve {keys [<at 0.6 1 RCIM_Cubic>, {arriveTangent=-2.6684, interpMode=RCIM_Cubic, leaveTangent=-2.6684, tangentMode=RCTM_User, time=1, value=0.5}]}}}
     Scale Curve=1.5
    )
    Scale RGB=&Multiply_VectorByFloat(
     Vector=&Lerp_Vector(
      A=&MakeVectorFromLinearColorRGB(Color=<rgba 1 0.8333 0 1>)
      B=&MakeVectorFromLinearColorRGB(Color=<rgba 0 1 0 1>)
      Alpha=&SelectFloatFromArray(
       Direct Array Index=&RandomRangeInt(Minimum=0, Maximum=1)
       Float Selection Array=DataInterface() {propertyValues {FloatData [0,1]}}
      )
     )
     Float=&FloatFromCurve(
      FloatCurve=DataInterface() {propertyValues {Curve {keys [
          <at 0 1 RCIM_Cubic>
          {arriveTangentWeight=0.3584, interpMode=RCIM_Cubic, tangentMode=RCTM_Break, tangentWeightMode=RCTWM_WeightedBoth, time=0.95, value=0.1}
         ]}}}
      Scale Curve=&RandomRangeFloat(Minimum=4, Maximum=0.2)
     )
    )
    ScaleA=true
    ScaleRGB=true
   ) {module=ScaleColor001}
   &InheritSourceMovement(
    Applied Position Scale=&VectorFromFloat(Value=&FloatFromCurve(FloatCurve=DataInterface() {propertyValues {Curve {keys [
          {interpMode=RCIM_Cubic, tangentMode=RCTM_Break, time=0, value=1}
          {arriveTangent=-2.9025, interpMode=RCIM_Cubic, tangentMode=RCTM_Break, time=1, value=0}
         ]}}}))
    ApplyForce=false
    ApplyPosition=true
    ApplyPositionFalloff=false
   ) {enabled=false}
   &SubUVAnimation(
    End Frame Range Override=9
    Evaluation Type=::NewEnumerator0
    "Sprite Renderer "=DataInterface() {propertyValues {SpriteRenderer=&self:NE_CS_MainTrail002_0.NiagaraSpriteRendererProperties_16}}
    Start Frame Range Override=1
    SubUV Animation Mode:ENiagara_SubUVLookupModeV2=::NewEnumerator2
    UseEndFrame=true
    UseStartFrame=true
   ) {module=SubUVAnimation001}
   &ScaleVelocity(
    Coordinate Space:ENiagaraCoordinateSpace=::World
    Velocity Scale=&MakeVector(
     X=1
     Y=1
     Z=&FloatFromCurve(
      FloatCurve=DataInterface() {propertyValues {Curve {keys [<at 0.2 0 RCIM_Cubic>, {arriveTangent=1.6052, interpMode=RCIM_Cubic, leaveTangent=1.6052, tangentMode=RCTM_User, time=0.9, value=1}]}}}
      Scale Curve=2
     )
    )
   ) {enabled=false}
   &Collision(
    Advanced Aging Rate=0
    Age Colliding Particles=true
    Analytical Collision Plane Position 1:NiagaraPosition=&VC_Mask_Vectors(Coro Value:Vector3f=<0 0 0>, FrontEnd Value:Vector3f=<0 0 -10>, FN Value:Vector3f=<-1200 0 0>)
    CPU Collision Type=::NewEnumerator3
    KillOnCollision=true
    Module.Control Rest State Via Surface Normal=false
    Module.EnableMaxCollisionCount=false
    Module.Randomize Collision Normal Vector=false
    Particle Radius Scale=0.5
   )
   &InheritSourceMovement(
    Applied Position Scale=&Multiply_VectorByFloat(
     Vector=<1 1 1>
     Float=&FloatFromCurve(FloatCurve=DataInterface() {propertyValues {Curve {
         defaultValue=1
         keys [
          {arriveTangent=-1.7885, interpMode=RCIM_Cubic, leaveTangent=-1.7885, tangentMode=RCTM_User, time=0.05, value=1}
          {interpMode=RCIM_Cubic, tangentMode=RCTM_User, time=1, value=0}
         ]
        }}})
    )
    ApplyForce=false
    ApplyPosition=true
    ApplyPositionFalloff=false
   ) {module=InheritSourceMovement001}
   &DynamicMaterialParameters(
    Index 0 Param 1=1
    Index 0 Param 2=1
    Index 0 Param 3=1
    Index 1 Param 1=1
    Index 1 Param 2=1
    Index 1 Param 3=&FloatFromCurve(
     FloatCurve=DataInterface() {propertyValues {Curve {keys [
         {
          arriveTangent=-0.1398
          interpMode=RCIM_Cubic
          leaveTangent=-0.366
          leaveTangentWeight=0.146
          tangentMode=RCTM_Break
          tangentWeightMode=RCTWM_WeightedBoth
          time=0
          value=0.16
         }
         {arriveTangentWeight=0.9374, interpMode=RCIM_Cubic, tangentMode=RCTM_Break, tangentWeightMode=RCTWM_WeightedBoth, time=1, value=0.05}
        ]}}}
     Scale Curve=30
    )
    Index 1 Param 4=1
    Index 2 Param 1=1
    Index 2 Param 3=1
    Index 3 Param 3=0.5
    Index 3 Param 4=1
    Module.Index0Float0Write=true
    Module.Index0Float1Write=true
    Module.Index0Float2Write=true
    Module.Index0Float3Write=true
    Module.Index1Float0Write=true
    Module.Index1Float1Write=true
    Module.Index1Float2Write=true
    Module.Index1Float3Write=true
    Module.Index2Float0Write=true
    Module.Index2Float1Write=true
    Module.Index2Float2Write=true
    Module.Index2Float3Write=true
    Module.Index3Float0Write=true
    Module.Index3Float1Write=true
    Module.Index3Float2Write=true
    Module.Index3Float3Write=true
    Write Parameter Index 1=::NewEnumerator1
    Write Parameter Index 2=::NewEnumerator1
    Write Parameter Index 3=::NewEnumerator1
   )
   &GravityForce(Gravity=<0 0 -800>)
   &SolveForcesAndVelocity(Clamp Velocity=false, Limit Acceleration=false)
   &GenerateDeathEvent(Use Event Probability=false)
  ]
  rendererClasses [&NiagaraSpriteRendererProperties]
  renderers [
   {rendererClass=&NiagaraSpriteRendererProperties}
   {rendererClass=&NiagaraSpriteRendererProperties, rendererIndex=1}
  ]
 }
 NE_VC_Boost_Ribbon_Parent {
  emitterUpdateScript [
   &EmitterState()
   &VC_FilteredSocketCount()
   &VC_RateQualityHandler(Default Rate=30)
   &VC_SpawnRibbon_v1(Number of Ribbons=$Emitter.PortCount, SpawnRate=$Emitter.Spawn Rate)
  ]
  particleSpawnScript [
   &VC_PaintedVariantsOverride() {enabled=false}
   &VC_InitializeRibbon_v1(
    Color=<rgba 3 0.8145 0.21 1>
    Custom Offset=true
    Enable Skeletal Position=true
    Lifetime=0.2
    Local Vector Offset=&AddVectorToPosition(Position=<0 0 0>, Vector=$User.DTCarOffset)
    Offset Amount=1
    Ribbon Width=120
   )
   &AddVelocity(Velocity=<-2500 0 50>) {module=AddVelocity001}
  ]
  particleUpdateScript [
   &ParticleState()
   &InheritSourceMovement(
    Applied Position Scale=&Multiply_VectorByFloat(
     Vector=<1 1 1>
     Float=&FloatFromCurve(FloatCurve=DataInterface() {propertyValues {Curve {keys [<at 0 1 RCIM_Cubic>, {interpMode=RCIM_Cubic, leaveTangent=-3.4326, time=0.7, value=0}]}}})
    )
    ApplyForce=false
    ApplyPosition=true
    ApplyPositionFalloff=false
   )
   &SolveForcesAndVelocity(Clamp Velocity=false, Limit Acceleration=false)
   &VC_UpdateRibbon_v1(Enable Skeletal Position=true)
   &ScaleColor(
    Scale Alpha=&FloatFromCurve(
     FloatCurve=DataInterface() {propertyValues {Curve {keys [
         {arriveTangent=8.6136, interpMode=RCIM_Cubic, leaveTangent=8.6136, tangentMode=RCTM_User, time=0, value=0}
         {interpMode=RCIM_Cubic, leaveTangentWeight=0.1978, tangentMode=RCTM_Break, tangentWeightMode=RCTWM_WeightedBoth, time=0.15, value=1}
         {arriveTangentWeight=0.7682, interpMode=RCIM_Cubic, tangentMode=RCTM_Break, tangentWeightMode=RCTWM_WeightedBoth, time=0.9, value=0}
        ]}}}
     Scale Curve=2
    )
    Scale RGB=<0.3 2 0.5>
    ScaleA=true
    ScaleRGB=true
   )
   &ScaleRibbonWidth(Ribbon Width Scale=&FloatFromCurve(FloatCurve=DataInterface() {propertyValues {Curve {keys [
         {arriveTangent=3.8014, interpMode=RCIM_Cubic, leaveTangent=3.8014, tangentMode=RCTM_User, time=0, value=0.7}
         {arriveTangent=-0.7, interpMode=RCIM_Cubic, leaveTangent=-0.7, time=0.2, value=1}
         <at 1 0>
        ]}}}))
  ]
  rendererClasses [&NiagaraRibbonRendererProperties]
  renderers [
   {rendererClass=&NiagaraRibbonRendererProperties}
   {rendererClass=&NiagaraRibbonRendererProperties, rendererIndex=1}
  ]
 }
 NE_VC_Boost_SPLASH001 {
  emitterUpdateScript [
   &EmitterState()
   &VC_FilteredSocketCount()
   &SpawnRate(SpawnRate=50, Use Spawn Probability=false) {enabled=false}
  ]
  particleSpawnScript [
   &VC_PaintedVariantsOverride(DefaultColor=<rgba 0.3472 2 0.0167 1>)
   &InitializeParticle(
    Lifetime Max=1
    Lifetime Min=0.3
    Lifetime Mode=::NewEnumerator1
    Sprite Rotation Mode=::NewEnumerator2
    Sprite Size Mode=::NewEnumerator1
    Uniform Sprite Size Max=&Multiply_Float(
     A=75
     B=&VC_Mask_Floats(Coro Value:NiagaraFloat=1, FrontEnd Value:NiagaraFloat=0.5, FN Value:NiagaraFloat=0.5)
    )
    Uniform Sprite Size Min=20
    UsePositionOffset=false
    Write Material Random=true
   )
   &PartitionParticles(Number Of Partitions=$Emitter.PortCount)
   &SkeletalMeshLocation(
    Bone / Socket Index=$Particles.PartitionParticles.Partition
    Bone / Socket Sampling Mode=::NewEnumerator5
    MaskSpawnGroup=false
    Mesh Sampling Type=::NewEnumerator6
    Skeletal Mesh=$System.SkeletalMesh
   )
   &AddVelocity(Velocity=<600 0 500>, Velocity Speed Scale=&RandomRangeFloat(Minimum=0.5, Maximum=1))
   &DLR_ProductPreviews(GameModeID:NiagaraInt32=$System.GameModeID, bPreviewVehicleSpeed:NiagaraBool=$System.bPreviewVehicleSpeed)
  ]
  particleUpdateScript [
   &ParticleState()
   &SpriteRotationRate(Rotation Rate=&RandomRangeFloat(Minimum=25, Maximum=-25, Evaluation Type=::NewEnumerator1))
   &ScaleSpriteSize(
    Uniform Curve Scale=&RandomRangeFloat(Minimum=4, Maximum=3.5)
    Uniform Curve Sprite Scale=DataInterface() {propertyValues {Curve {keys [
        {arriveTangent=4.1785, interpMode=RCIM_Cubic, leaveTangent=4.1785, tangentMode=RCTM_User, time=0, value=0}
        {arriveTangent=1.1179, interpMode=RCIM_Cubic, leaveTangent=1.1179, tangentMode=RCTM_User, time=0.1, value=0.8}
        {arriveTangent=-0.3333, interpMode=RCIM_Cubic, leaveTangent=-0.3333, time=0.7, value=1}
        {arriveTangent=1.8609, interpMode=RCIM_Cubic, leaveTangent=1.8609, tangentMode=RCTM_User, time=1, value=0.5}
       ]}}}
   )
   &AddRotationalVelocity(Rotation Rate=&RandomRangeVector(Minimum=<0 0 50>, Maximum=<1 0 1>))
   &GravityForce(Gravity=<0 0 -1200>)
   &InheritSourceMovement(
    Applied Position Scale=&Multiply_VectorByFloat(
     Vector=<1 1 1>
     Float=&FloatFromCurve(FloatCurve=DataInterface() {propertyValues {Curve {
         defaultValue=1
         keys [
          {arriveTangent=-1.7885, interpMode=RCIM_Cubic, leaveTangent=-1.7885, tangentMode=RCTM_User, time=0.05, value=1}
          {interpMode=RCIM_Cubic, tangentMode=RCTM_User, time=1, value=0}
         ]
        }}})
    )
    ApplyForce=false
    ApplyPosition=true
    ApplyPositionFalloff=false
   ) {enabled=false, module=InheritSourceMovement001}
   &SubUVAnimation(
    Evaluation Type=::NewEnumerator0
    "Sprite Renderer "=DataInterface() {propertyValues {SpriteRenderer=&self:NE_VC_Boost_SPLASH001_6.NiagaraSpriteRendererProperties_1}}
    Start Frame Range Override=2
    SubUV Animation Mode:ENiagara_SubUVLookupModeV2=::NewEnumerator2
    UseEndFrame=false
    UseStartFrame=false
   )
   &ShapeLocation(
    MaskSpawnGroup=false
    Offset=<10 0 5>
    Offset Mode=::NewEnumerator0
    Override Seed=false
    Shape Origin=$Particles.Position
    Sphere Radius=15
    UseOwnerScale=true
   )
   &ScaleColor(
    Scale Alpha=&FloatFromCurve(
     FloatCurve=DataInterface() {propertyValues {Curve {keys [
         {arriveTangent=0.0061, interpMode=RCIM_Cubic, leaveTangent=0.0061, tangentMode=RCTM_User, time=0.5, value=1}
         {interpMode=RCIM_Cubic, tangentMode=RCTM_User, time=0.8, value=0}
        ]}}}
     Scale Curve=&RandomRangeFloat(Minimum=0.5, Maximum=1)
    )
    Scale RGB=&Multiply_VectorByFloat(
     Vector=&MakeVectorFromLinearColorRGB(Color=&ColorFromCurve(ColorCurve=DataInterface() {propertyValues {
         AlphaCurve {keys [<at 0 1 RCIM_Cubic>,<at 1 0 RCIM_Cubic>]}
         BlueCurve {keys [<at 0.0769 0>,<at 0.3343 0 RCIM_Cubic>,<at 0.7296 0 RCIM_Cubic>]}
         GreenCurve {keys [<at 0.0769 20>, {arriveTangent=-30.6267, interpMode=RCIM_Cubic, leaveTangent=-30.6267, time=0.3343, value=5}, <at 0.7296 0.01 RCIM_Cubic>]}
         RedCurve {keys [
           <at 0.0769 6.6667>
           {arriveTangent=-10.2089, interpMode=RCIM_Cubic, leaveTangent=-10.2089, time=0.3343, value=1.25}
           <at 0.7296 0.0033 RCIM_Cubic>
          ]}
        }}))
     Float=&RandomRangeFloat(Minimum=0.001, Maximum=1)
    )
    ScaleA=true
    ScaleRGB=true
   )
   &SolveForcesAndVelocity(Clamp Velocity=false, Limit Acceleration=false)
  ]
  rendererClasses [&NiagaraSpriteRendererProperties]
  renderers [{rendererClass=&NiagaraSpriteRendererProperties}]
 }
 NE_VC_GLow {
  emitterUpdateScript [
   &EmitterState()
   &VC_FilteredSocketCount()
   &SpawnRate(SpawnRate=180, Use Spawn Probability=false)
  ]
  particleSpawnScript [
   /SetVariables(Particles.VisTag:NiagaraInt32=&SpawnGroup()) {isSetParameters=true, module=SetVariables_#4}
   &VC_PaintedVariantsOverride(
    R2 Black:LinearColor=<rgba 0.16 0.16 0.16 1>
    R2 DefaultColor:LinearColor=<rgba 0.9348 0.5504 0.9896 1>
    Use R1 Painted Overrides=true
    Use R2 Painted Overrides=true
   ) {enabled=false}
   &InitializeParticle(
    Lifetime Max=0.5
    Lifetime Min=0.6
    Lifetime Mode=::NewEnumerator1
    Sprite Rotation Mode=::NewEnumerator2
    Sprite Size Mode=::NewEnumerator1
    Uniform Sprite Size Max=190
    Uniform Sprite Size Min=100
    UsePositionOffset=false
    Write Material Random=true
   )
   &PartitionParticles(Number Of Partitions=$Emitter.PortCount)
   &SkeletalMeshLocation(
    Bone / Socket Index=$Particles.PartitionParticles.Partition
    Bone / Socket Sampling Mode=::NewEnumerator5
    MaskSpawnGroup=false
    Mesh Sampling Type=::NewEnumerator6
    Sampled Position Offset=&AddVectorToPosition(Position=<0 0 0>, Vector=$User.DTCarOffset)
    Skeletal Mesh=$System.SkeletalMesh
   )
   &AddVelocity(Velocity=<-1500 0 -300>, Velocity Speed Scale=1.13)
   &DLR_ProductPreviews(GameModeID:NiagaraInt32=$System.GameModeID, bPreviewVehicleSpeed:NiagaraBool=$System.bPreviewVehicleSpeed)
  ]
  particleUpdateScript [
   &ParticleState()
   /SetVariables(
    Particles.Glow_Color:LinearColor=&ColorFromCurve(ColorCurve=DataInterface() {propertyValues {
       AlphaCurve {keys [<at 0 0>,<at 0.0459 1 RCIM_Cubic>,<at 0.6733 0 RCIM_Cubic>]}
       BlueCurve {keys [<at 0 0 RCIM_Cubic>]}
       GreenCurve {keys [<at 0 0.7374 RCIM_Cubic>]}
       RedCurve {keys [<at 0 2 RCIM_Cubic>]}
      }})
    Particles.Glow_Size:Vector2f=&RandomRangeVector2D(Minimum=<100 100>, Maximum=<200 200>)
   ) {enabled=false, isSetParameters=true, module=SetVariables_#5}
   &VC_PaintedVariantsOverride(
    Black=&ColorFromCurve(ColorCurve=DataInterface() {propertyValues {
       AlphaCurve {keys [<at 0.0801 0 RCIM_Cubic>,<at 0.119 1 RCIM_Cubic>]}
       BlueCurve {keys [<at 0.1205 0.01 RCIM_Cubic>]}
       GreenCurve {keys [<at 0.1205 0.01 RCIM_Cubic>]}
       RedCurve {keys [<at 0.1205 0.01 RCIM_Cubic>]}
      }})
    BurntSienna=&ColorFromCurve(ColorCurve=DataInterface() {propertyValues {
       AlphaCurve {keys [<at 0.0646 0>,<at 0.1641 1 RCIM_Cubic>,<at 1 0 RCIM_Cubic>]}
       BlueCurve {keys [<at 0 0.0075 RCIM_Cubic>,<at 0.6548 0.002 RCIM_Cubic>]}
       GreenCurve {keys [<at 0 0.0382 RCIM_Cubic>,<at 0.6548 0.0367 RCIM_Cubic>]}
       RedCurve {keys [<at 0 0.2346 RCIM_Cubic>,<at 0.6548 0.0305 RCIM_Cubic>]}
      }})
    Cobalt=&ColorFromCurve(ColorCurve=DataInterface() {propertyValues {
       AlphaCurve {keys [<at 0.0614 0 RCIM_Cubic>,<at 0.1719 1 RCIM_Cubic>,<at 0.6292 0.0028>]}
       BlueCurve {keys [<at 0 1 RCIM_Cubic>,<at 0.8633 0 RCIM_Cubic>]}
       GreenCurve {keys [<at 0 0 RCIM_Cubic>,<at 0.8633 0 RCIM_Cubic>]}
       RedCurve {keys [<at 0 0.0003 RCIM_Cubic>,<at 0.8633 0 RCIM_Cubic>]}
      }})
    Crimson=&ColorFromCurve(ColorCurve=DataInterface() {propertyValues {
       AlphaCurve {keys [<at 0.0381 0>,<at 0.1921 1 RCIM_Cubic>,<at 0.6836 1>,<at 1 0 RCIM_Cubic>]}
       BlueCurve {keys [<at 0 0.0033 RCIM_Cubic>,<at 0.1921 0.0125>,<at 0.4736 0.0234 RCIM_Cubic>]}
       GreenCurve {keys [<at 0 0.0033 RCIM_Cubic>,<at 0.1921 0.0257>,<at 0.4736 0.3926 RCIM_Cubic>]}
       RedCurve {keys [<at 0 0.1875 RCIM_Cubic>,<at 0.1921 0.6771>,<at 0.4736 0.4948 RCIM_Cubic>]}
      }})
    DefaultColor=&ColorFromCurve(ColorCurve=DataInterface() {propertyValues {
       AlphaCurve {keys [<at 0 0 RCIM_Cubic>,<at 0.1719 1 RCIM_Cubic>]}
       BlueCurve {keys [<at 0.1753 0.5156 RCIM_Cubic>,<at 0.6077 1>]}
       GreenCurve {keys [<at 0.1753 0.2305 RCIM_Cubic>,<at 0.6077 0.3073>]}
       RedCurve {keys [<at 0.1753 0.2585 RCIM_Cubic>,<at 0.6077 0.8652>]}
      }})
    ForestGreen=&ColorFromCurve(ColorCurve=DataInterface() {propertyValues {
       AlphaCurve {keys [<at 0.0381 0>,<at 0.1921 1 RCIM_Cubic>,<at 0.6836 1>,<at 1 0 RCIM_Cubic>]}
       BlueCurve {keys [<at 0 0.0856 RCIM_Cubic>,<at 0.259 0.0804>,<at 0.4301 0.5573 RCIM_Cubic>]}
       GreenCurve {keys [<at 0 0.5826 RCIM_Cubic>,<at 0.259 0.5469>,<at 0.4301 0.5279 RCIM_Cubic>]}
       RedCurve {keys [<at 0 0 RCIM_Cubic>,<at 0.259 0>,<at 0.4301 0 RCIM_Cubic>]}
      }})
    Grey=&ColorFromCurve(ColorCurve=DataInterface() {propertyValues {
       AlphaCurve {keys [<at 0.0303 0 RCIM_Cubic>,<at 0.1439 1 RCIM_Cubic>]}
       BlueCurve {keys [<at 0 0.0999 RCIM_Cubic>,<at 0.2745 0.099 RCIM_Cubic>]}
       GreenCurve {keys [<at 0 0.0999 RCIM_Cubic>,<at 0.2745 0.099 RCIM_Cubic>]}
       RedCurve {keys [<at 0 0.0999 RCIM_Cubic>,<at 0.2745 0.099 RCIM_Cubic>]}
      }})
    Lime=&ColorFromCurve(ColorCurve=DataInterface() {propertyValues {
       AlphaCurve {keys [
         <at 0.0506 0 RCIM_Cubic>
         {arriveTangent=2.6348, interpMode=RCIM_Cubic, leaveTangent=2.6348, time=0.1859, value=1}
         <at 0.4301 1>
         <at 0.6619 0>
        ]}
       BlueCurve {keys [<at 0.1828 0 RCIM_Cubic>,<at 0.3834 0.061 RCIM_Cubic>]}
       GreenCurve {keys [<at 0.1828 0.7917 RCIM_Cubic>,<at 0.3834 0.1476 RCIM_Cubic>]}
       RedCurve {keys [<at 0.1828 0.1696 RCIM_Cubic>,<at 0.3834 0.0081 RCIM_Cubic>]}
      }})
    Orange=&ColorFromCurve(ColorCurve=DataInterface() {propertyValues {
       AlphaCurve {keys [<at 0 0 RCIM_Cubic>, {arriveTangent=2.6348, interpMode=RCIM_Cubic, time=0.1859, value=1}, <at 0.4301 1>, <at 0.6619 0>]}
       BlueCurve {keys [<at 0.1828 0.0008 RCIM_Cubic>,<at 0.3834 0 RCIM_Cubic>]}
       GreenCurve {keys [<at 0.1828 0.2332 RCIM_Cubic>,<at 0.3834 0.3646 RCIM_Cubic>]}
       RedCurve {keys [<at 0.1828 0.6146 RCIM_Cubic>,<at 0.3834 0.2732 RCIM_Cubic>]}
      }})
    Pink=&ColorFromCurve(ColorCurve=DataInterface() {propertyValues {
       AlphaCurve {keys [
         <at 0 0 RCIM_Cubic>
         {arriveTangent=2.3251, interpMode=RCIM_Cubic, leaveTangent=2.3251, time=0.1859, value=0.8}
         <at 0.4301 1>
         <at 0.6034 0>
        ]}
       BlueCurve {keys [<at 0.1828 0.147 RCIM_Cubic>,<at 0.5031 0.1628 RCIM_Cubic>]}
       GreenCurve {keys [<at 0.1828 0.0821 RCIM_Cubic>,<at 0.5031 0.1997 RCIM_Cubic>]}
       RedCurve {keys [<at 0.1828 0.1771 RCIM_Cubic>,<at 0.5031 0.3906 RCIM_Cubic>]}
      }})
    Purple=&ColorFromCurve(ColorCurve=DataInterface() {propertyValues {
       AlphaCurve {keys [
         <at 0 0 RCIM_Cubic>
         {arriveTangent=2.3251, interpMode=RCIM_Cubic, leaveTangent=2.3251, time=0.1099, value=0.8}
         <at 0.4301 1>
         <at 0.6034 0>
        ]}
       BlueCurve {keys [<at 0.1828 0.1771 RCIM_Cubic>,<at 0.5031 0.5648 RCIM_Cubic>]}
       GreenCurve {keys [<at 0.1828 0.0338 RCIM_Cubic>,<at 0.5031 0.0011 RCIM_Cubic>]}
       RedCurve {keys [<at 0.1828 0.138 RCIM_Cubic>,<at 0.5031 0.5885 RCIM_Cubic>]}
      }})
    R1 Black:LinearColor=&ColorFromCurve(ColorCurve=DataInterface() {propertyValues {
       AlphaCurve {keys [<at 0.1097 0 RCIM_Cubic>,<at 0.2745 1 RCIM_Cubic>,<at 0.6836 0>]}
       BlueCurve {keys [<at 0.1091 0.01 RCIM_Cubic>,<at 0.2892 0.1771 RCIM_Cubic>]}
       GreenCurve {keys [<at 0.1091 0.01 RCIM_Cubic>,<at 0.2892 0.1771 RCIM_Cubic>]}
       RedCurve {keys [<at 0.1091 0.01 RCIM_Cubic>,<at 0.2892 0.1771 RCIM_Cubic>]}
      }})
    R1 BurntSienna:LinearColor=&ColorFromCurve(ColorCurve=DataInterface() {propertyValues {
       AlphaCurve {keys [<at 0 0>,<at 0.469 1 RCIM_Cubic>,<at 0.6525 0 RCIM_Cubic>]}
       BlueCurve {keys [<at 0 0.0178 RCIM_Cubic>,<at 0.6548 0.0108 RCIM_Cubic>]}
       GreenCurve {keys [<at 0 0.0908 RCIM_Cubic>,<at 0.6548 0.6927 RCIM_Cubic>]}
       RedCurve {keys [<at 0 0.5573 RCIM_Cubic>,<at 0.6548 0.57 RCIM_Cubic>]}
      }})
    R1 Cobalt:LinearColor=&ColorFromCurve(ColorCurve=DataInterface() {propertyValues {
       AlphaCurve {keys [<at 0.1174 0 RCIM_Cubic>,<at 0.2528 1 RCIM_Cubic>,<at 0.6105 0>]}
       BlueCurve {keys [<at 0.1753 1 RCIM_Cubic>,<at 0.6077 3>]}
       GreenCurve {keys [<at 0.1753 0.4069 RCIM_Cubic>,<at 0.6077 0>]}
       RedCurve {keys [<at 0.1753 0.3646 RCIM_Cubic>,<at 0.6077 2.5078>]}
      }})
    R1 Crimson:LinearColor=&ColorFromCurve(ColorCurve=DataInterface() {propertyValues {
       AlphaCurve {keys [
         <at 0.147 0 RCIM_Cubic>
         {arriveTangent=3.9441, interpMode=RCIM_Cubic, leaveTangent=3.9441, time=0.2808, value=1}
         <at 0.4005 1>
         <at 0.6074 0>
        ]}
       BlueCurve {keys [<at 0.1753 0.1286 RCIM_Cubic>,<at 0.6077 0>]}
       GreenCurve {keys [<at 0.1753 0.1763 RCIM_Cubic>,<at 0.6077 1.4989>]}
       RedCurve {keys [<at 0.1753 1 RCIM_Cubic>,<at 0.6077 1.5>]}
      }})
    R1 DefaultColor=&ColorFromCurve(ColorCurve=DataInterface() {propertyValues {
       AlphaCurve {keys [<at 0.0817 0 RCIM_Cubic>,<at 0.2357 1 RCIM_Cubic>,<at 0.6821 0>]}
       BlueCurve {keys [<at 0.1753 1 RCIM_Cubic>,<at 0.6077 2>]}
       GreenCurve {keys [<at 0.1753 0.8125 RCIM_Cubic>,<at 0.6077 0.6875>]}
       RedCurve {keys [<at 0.1753 0.831 RCIM_Cubic>,<at 0.6077 1.7446>]}
      }})
    R1 ForestGreen:LinearColor=&ColorFromCurve(ColorCurve=DataInterface() {propertyValues {
       AlphaCurve {keys [<at 0.0739 0 RCIM_Cubic>, {arriveTangent=3.9441, interpMode=RCIM_Cubic, time=0.2808, value=1}, <at 0.4005 1>, <at 0.6074 0>]}
       BlueCurve {keys [{arriveTangent=3.2911, interpMode=RCIM_Cubic, time=0.1753, value=0}, <at 0.6077 2>]}
       GreenCurve {keys [{arriveTangent=2.7598, interpMode=RCIM_Cubic, time=0.1753, value=0.8594}, <at 0.6077 2>]}
       RedCurve {keys [{arriveTangent=-0.1276, interpMode=RCIM_Cubic, time=0.1753, value=0.2064}, <at 0.6077 0>]}
      }})
    R1 Grey:LinearColor=&ColorFromCurve(ColorCurve=DataInterface() {propertyValues {
       AlphaCurve {keys [<at 0.1143 0>,<at 0.3088 1 RCIM_Cubic>,<at 0.6043 0 RCIM_Cubic>]}
       BlueCurve {keys [<at 0.3072 0.6562 RCIM_Cubic>]}
       GreenCurve {keys [<at 0.3072 0.6562 RCIM_Cubic>]}
       RedCurve {keys [<at 0.3072 0.6562 RCIM_Cubic>]}
      }})
    R1 Lime:LinearColor=&ColorFromCurve(ColorCurve=DataInterface() {propertyValues {
       AlphaCurve {keys [<at 0.0739 0 RCIM_Cubic>, {arriveTangent=3.9441, interpMode=RCIM_Cubic, time=0.2808, value=1}, <at 0.4005 1>, <at 0.6074 0>]}
       BlueCurve {keys [<at 0.091 0 RCIM_Cubic>,<at 0.6077 1.3543>]}
       GreenCurve {keys [<at 0.091 2 RCIM_Cubic>,<at 0.6077 2>]}
       RedCurve {keys [<at 0.091 1.1387 RCIM_Cubic>,<at 0.6077 0>]}
      }})
    R1 Orange:LinearColor=&ColorFromCurve(ColorCurve=DataInterface() {propertyValues {
       AlphaCurve {keys [<at 0.0739 0 RCIM_Cubic>, {arriveTangent=3.9441, interpMode=RCIM_Cubic, time=0.2808, value=1}, <at 0.4005 1>, <at 0.6074 0>]}
       BlueCurve {keys [<at 0.091 0 RCIM_Cubic>,<at 0.6077 0.2815>]}
       GreenCurve {keys [<at 0.091 0.7377 RCIM_Cubic>,<at 0.6077 1>]}
       RedCurve {keys [<at 0.091 2 RCIM_Cubic>,<at 0.6077 0.7462>]}
      }})
    R1 Pink:LinearColor=&ColorFromCurve(ColorCurve=DataInterface() {propertyValues {
       AlphaCurve {keys [<at 0.0739 0 RCIM_Cubic>, {arriveTangent=3.9441, interpMode=RCIM_Cubic, time=0.2808, value=1}, <at 0.4005 1>, <at 0.6074 0>]}
       BlueCurve {keys [<at 0.091 1.7582 RCIM_Cubic>,<at 0.3276 0.4062>]}
       GreenCurve {keys [<at 0.091 0.1064 RCIM_Cubic>,<at 0.3276 0.8298>]}
       RedCurve {keys [<at 0.091 2 RCIM_Cubic>,<at 0.3276 2>]}
      }})
    R1 Purple:LinearColor=&ColorFromCurve(ColorCurve=DataInterface() {propertyValues {
       AlphaCurve {keys [<at 0.1131 0 RCIM_Cubic>, {arriveTangent=3.9441, interpMode=RCIM_Cubic, time=0.2466, value=1}, <at 0.4005 1>, <at 0.6074 0>]}
       BlueCurve {keys [<at 0.2037 5 RCIM_Cubic>,<at 0.3276 0.709>]}
       GreenCurve {keys [<at 0.2037 0 RCIM_Cubic>,<at 0.3276 0.1062>]}
       RedCurve {keys [<at 0.2037 0.9897 RCIM_Cubic>,<at 0.3276 1>]}
      }})
    R1 Saffron:LinearColor=&ColorFromCurve(ColorCurve=DataInterface() {propertyValues {
       AlphaCurve {keys [<at 0.1131 0 RCIM_Cubic>, {arriveTangent=3.9441, interpMode=RCIM_Cubic, time=0.2466, value=1}, <at 0.4005 1>, <at 0.6074 0>]}
       BlueCurve {keys [<at 0.1542 0.875 RCIM_Cubic>,<at 0.3276 0.533>]}
       GreenCurve {keys [<at 0.1542 2.7603 RCIM_Cubic>,<at 0.3276 2>]}
       RedCurve {keys [<at 0.1542 3 RCIM_Cubic>,<at 0.3276 0.9975>]}
      }})
    R1 SkyBlue:LinearColor=&ColorFromCurve(ColorCurve=DataInterface() {propertyValues {
       AlphaCurve {keys [<at 0.1131 0 RCIM_Cubic>, {arriveTangent=3.9441, interpMode=RCIM_Cubic, time=0.2466, value=1}, <at 0.4005 1>, <at 0.6074 0>]}
       BlueCurve {keys [<at 0.1542 2 RCIM_Cubic>,<at 0.3276 1>]}
       GreenCurve {keys [<at 0.1542 1.715 RCIM_Cubic>,<at 0.3276 0.3668>]}
       RedCurve {keys [<at 0.1542 0.0312 RCIM_Cubic>,<at 0.3276 0.519>]}
      }})
    R1 TitaniumWhite:LinearColor=&ColorFromCurve(ColorCurve=DataInterface() {propertyValues {
       AlphaCurve {keys [<at 0.1131 0 RCIM_Cubic>, {arriveTangent=3.9441, interpMode=RCIM_Cubic, time=0.2466, value=1}, <at 0.4005 1>, <at 0.6074 0>]}
       BlueCurve {keys [<at 0.1542 2 RCIM_Cubic>,<at 0.3276 2>]}
       GreenCurve {keys [<at 0.1542 2 RCIM_Cubic>,<at 0.3276 2>]}
       RedCurve {keys [<at 0.1542 2 RCIM_Cubic>,<at 0.3276 2>]}
      }})
    R2 Black:LinearColor=&ColorFromCurve(ColorCurve=DataInterface() {propertyValues {
       AlphaCurve {keys [<at 0.0677 0 RCIM_Cubic>,<at 0.2668 0.8 RCIM_Cubic>,<at 1 0>]}
       BlueCurve {keys [<at 0.1091 0.001 RCIM_Cubic>,<at 0.2892 0.02 RCIM_Cubic>]}
       GreenCurve {keys [<at 0.1091 0.001 RCIM_Cubic>,<at 0.2892 0.02 RCIM_Cubic>]}
       RedCurve {keys [<at 0.1091 0.001 RCIM_Cubic>,<at 0.2892 0.02 RCIM_Cubic>]}
      }})
    R2 BurntSienna:LinearColor=&ColorFromCurve(ColorCurve=DataInterface() {propertyValues {
       AlphaCurve {keys [<at 0.1765 0 RCIM_Cubic>,<at 0.4052 0.6 RCIM_Cubic>]}
       BlueCurve {keys [<at 0.2419 0.0062 RCIM_Cubic>,<at 0.3865 0.0191 RCIM_Cubic>]}
       GreenCurve {keys [<at 0.2419 0.0287 RCIM_Cubic>,<at 0.3865 0.0573 RCIM_Cubic>]}
       RedCurve {keys [<at 0.2419 0.1771 RCIM_Cubic>,<at 0.3865 0.0504 RCIM_Cubic>]}
      }})
    R2 Cobalt:LinearColor=&ColorFromCurve(ColorCurve=DataInterface() {propertyValues {
       AlphaCurve {keys [<at 0.0817 0 RCIM_Cubic>,<at 0.4239 0.6 RCIM_Cubic>,<at 0.8485 0>]}
       BlueCurve {keys [<at 0.1753 0.3281 RCIM_Cubic>,<at 0.4954 0.224>]}
       GreenCurve {keys [<at 0.1753 0.0218 RCIM_Cubic>,<at 0.4954 0>]}
       RedCurve {keys [<at 0.1753 0 RCIM_Cubic>,<at 0.4954 0.1872>]}
      }})
    R2 Crimson:LinearColor=&ColorFromCurve(ColorCurve=DataInterface() {propertyValues {
       AlphaCurve {keys [<at 0.0817 0 RCIM_Cubic>,<at 0.3492 0.8 RCIM_Cubic>,<at 0.8485 0>]}
       BlueCurve {keys [<at 0.1753 0 RCIM_Cubic>,<at 0.4954 0.0137>]}
       GreenCurve {keys [<at 0.1753 0.0039 RCIM_Cubic>,<at 0.4954 0.0729>]}
       RedCurve {keys [<at 0.1753 0.2083 RCIM_Cubic>,<at 0.4954 0.0566>]}
      }})
    R2 DefaultColor:LinearColor=&ColorFromCurve(ColorCurve=DataInterface() {propertyValues {
       AlphaCurve {keys [<at 0.1672 0 RCIM_Cubic>,<at 0.3803 0.7 RCIM_Cubic>,<at 0.7334 0>]}
       BlueCurve {keys [<at 0.1753 0.1198 RCIM_Cubic>,<at 0.3508 0.1771>]}
       GreenCurve {keys [<at 0.1753 0.0381 RCIM_Cubic>,<at 0.3508 0.0839>]}
       RedCurve {keys [<at 0.1753 0.0461 RCIM_Cubic>,<at 0.3508 0.159>]}
      }})
    R2 ForestGreen:LinearColor=&ColorFromCurve(ColorCurve=DataInterface() {propertyValues {
       AlphaCurve {keys [
         <at 0.049 0>
         {arriveTangent=1.2606, interpMode=RCIM_Cubic, leaveTangent=1.2606, time=0.4161, value=0.8}
         <at 0.6836 0.8>
         <at 1 0 RCIM_Cubic>
        ]}
       BlueCurve {keys [<at 0 0.0334>,<at 0.4736 0.0922 RCIM_Cubic>]}
       GreenCurve {keys [<at 0 0.1823>,<at 0.4736 0.1198 RCIM_Cubic>]}
       RedCurve {keys [<at 0 0>,<at 0.4736 0 RCIM_Cubic>]}
      }})
    R2 Grey:LinearColor=&ColorFromCurve(ColorCurve=DataInterface() {propertyValues {
       AlphaCurve {keys [
         <at 0.147 0 RCIM_Cubic>
         {arriveTangent=3.1553, interpMode=RCIM_Cubic, leaveTangent=3.1553, time=0.2808, value=0.8}
         <at 0.4005 0.8>
         <at 0.6074 0>
        ]}
       BlueCurve {keys [<at 0.0708 0.2083 RCIM_Cubic>,<at 0.3508 0.0833>]}
       GreenCurve {keys [<at 0.0708 0.2083 RCIM_Cubic>,<at 0.3508 0.0833>]}
       RedCurve {keys [<at 0.0708 0.2083 RCIM_Cubic>,<at 0.3508 0.0833>]}
      }})
    R2 Lime:LinearColor=&ColorFromCurve(ColorCurve=DataInterface() {propertyValues {
       AlphaCurve {keys [
         <at 0.0412 0 RCIM_Cubic>
         {arriveTangent=1.6698, interpMode=RCIM_Cubic, leaveTangent=1.6698, time=0.2808, value=0.6}
         <at 0.4005 0.6>
         <at 0.6074 0>
        ]}
       BlueCurve {keys [<at 0.0941 0.0002 RCIM_Cubic>,<at 0.3803 0.1206>]}
       GreenCurve {keys [<at 0.0941 0.3438 RCIM_Cubic>,<at 0.3803 0.1927>]}
       RedCurve {keys [<at 0.0941 0.2305 RCIM_Cubic>,<at 0.3803 0.0384>]}
      }})
    R2 Orange:LinearColor=&ColorFromCurve(ColorCurve=DataInterface() {propertyValues {
       AlphaCurve {keys [
         <at 0.0412 0 RCIM_Cubic>
         {arriveTangent=1.6698, interpMode=RCIM_Cubic, leaveTangent=1.6698, time=0.1552, value=0.5}
         <at 0.4005 0.6>
         <at 0.6074 0>
        ]}
       BlueCurve {keys [<at 0.1583 0 RCIM_Cubic>,<at 0.3803 0>]}
       GreenCurve {keys [<at 0.1583 0.2274 RCIM_Cubic>,<at 0.3803 0.1458>]}
       RedCurve {keys [<at 0.1583 0.5469 RCIM_Cubic>,<at 0.3803 0.1076>]}
      }})
    R2 Pink:LinearColor=&ColorFromCurve(ColorCurve=DataInterface() {propertyValues {
       AlphaCurve {keys [
         <at 0.0412 0 RCIM_Cubic>
         {arriveTangent=1.6698, interpMode=RCIM_Cubic, leaveTangent=1.6698, time=0.1552, value=0.5}
         <at 0.4005 0.6>
         <at 0.6074 0>
        ]}
       BlueCurve {keys [<at 0.1583 0.2922 RCIM_Cubic>,<at 0.3803 0.0526>]}
       GreenCurve {keys [<at 0.1583 0.0423 RCIM_Cubic>,<at 0.3803 0.0787>]}
       RedCurve {keys [<at 0.1583 0.4271 RCIM_Cubic>,<at 0.3803 0.1094>]}
      }})
    R2 Purple:LinearColor=&ColorFromCurve(ColorCurve=DataInterface() {propertyValues {
       AlphaCurve {keys [
         <at 0.0412 0 RCIM_Cubic>
         {arriveTangent=1.6698, interpMode=RCIM_Cubic, leaveTangent=1.6698, time=0.1552, value=0.5}
         <at 0.4005 0.6>
         <at 0.6074 0>
        ]}
       BlueCurve {keys [<at 0.1583 1 RCIM_Cubic>,<at 0.3803 0.08>]}
       GreenCurve {keys [<at 0.1583 0 RCIM_Cubic>,<at 0.3803 0.0166>]}
       RedCurve {keys [<at 0.1583 0.287 RCIM_Cubic>,<at 0.3803 0.1094>]}
      }})
    R2 Saffron:LinearColor=&ColorFromCurve(ColorCurve=DataInterface() {propertyValues {
       AlphaCurve {keys [
         <at 0.0412 0 RCIM_Cubic>
         {arriveTangent=1.6698, interpMode=RCIM_Cubic, leaveTangent=1.6698, time=0.1552, value=0.5}
         <at 0.4005 0.6>
         <at 0.6074 0>
        ]}
       BlueCurve {keys [<at 0.1583 0 RCIM_Cubic>,<at 0.3803 0.0873>]}
       GreenCurve {keys [<at 0.1583 0.8117 RCIM_Cubic>,<at 0.3803 0.2396>]}
       RedCurve {keys [<at 0.1583 0.875 RCIM_Cubic>,<at 0.3803 0.1074>]}
      }})
    R2 SkyBlue:LinearColor=&ColorFromCurve(ColorCurve=DataInterface() {propertyValues {
       AlphaCurve {keys [
         <at 0.0412 0 RCIM_Cubic>
         {arriveTangent=1.6698, interpMode=RCIM_Cubic, leaveTangent=1.6698, time=0.1552, value=0.5}
         <at 0.4005 0.6>
         <at 0.6074 0>
        ]}
       BlueCurve {keys [<at 0.1583 0.4167 RCIM_Cubic>,<at 0.3803 0.125>]}
       GreenCurve {keys [<at 0.1583 0.2756 RCIM_Cubic>,<at 0.3803 0.041>]}
       RedCurve {keys [<at 0.1583 0.0434 RCIM_Cubic>,<at 0.3803 0.0705>]}
      }})
    R2 TitaniumWhite:LinearColor=&ColorFromCurve(ColorCurve=DataInterface() {propertyValues {
       AlphaCurve {keys [
         <at 0.0412 0 RCIM_Cubic>
         {arriveTangent=1.6698, interpMode=RCIM_Cubic, leaveTangent=1.6698, time=0.1552, value=0.5}
         <at 0.4005 0.6>
         <at 0.6074 0>
        ]}
       BlueCurve {keys [<at 0.1583 0.6667 RCIM_Cubic>,<at 0.3803 0.2031>]}
       GreenCurve {keys [<at 0.1583 0.6667 RCIM_Cubic>,<at 0.3803 0.2031>]}
       RedCurve {keys [<at 0.1583 0.6667 RCIM_Cubic>,<at 0.3803 0.2031>]}
      }})
    Saffron=&ColorFromCurve(ColorCurve=DataInterface() {propertyValues {
       AlphaCurve {keys [<at 0.0318 0 RCIM_Cubic>, {arriveTangent=2.3251, interpMode=RCIM_Cubic, time=0.1803, value=0.8}, <at 0.4301 1>, <at 0.6034 0>]}
       BlueCurve {keys [<at 0.1258 0 RCIM_Cubic>,<at 0.3258 0 RCIM_Cubic>]}
       GreenCurve {keys [<at 0.1258 0.7424 RCIM_Cubic>,<at 0.3258 0.4219 RCIM_Cubic>]}
       RedCurve {keys [<at 0.1258 0.875 RCIM_Cubic>,<at 0.3258 0.2191 RCIM_Cubic>]}
      }})
    SkyBlue=&ColorFromCurve(ColorCurve=DataInterface() {propertyValues {
       AlphaCurve {keys [<at 0.0437 0 RCIM_Cubic>, {arriveTangent=2.3251, interpMode=RCIM_Cubic, time=0.1654, value=0.8}, <at 0.4301 1>, <at 0.6034 0>]}
       BlueCurve {keys [<at 0.1022 0.4792 RCIM_Cubic>,<at 0.3602 0.4323 RCIM_Cubic>]}
       GreenCurve {keys [<at 0.1022 0.4551 RCIM_Cubic>,<at 0.3602 0.1302 RCIM_Cubic>]}
       RedCurve {keys [<at 0.1022 0.0393 RCIM_Cubic>,<at 0.3602 0.2357 RCIM_Cubic>]}
      }})
    TitaniumWhite=&ColorFromCurve(ColorCurve=DataInterface() {propertyValues {
       AlphaCurve {keys [<at 0.0437 0 RCIM_Cubic>, {arriveTangent=2.3251, interpMode=RCIM_Cubic, time=0.1654, value=0.8}, <at 0.4301 1>, <at 0.6034 0>]}
       BlueCurve {keys [<at 0.1022 1 RCIM_Cubic>,<at 0.3602 0.4323 RCIM_Cubic>]}
       GreenCurve {keys [<at 0.1022 1 RCIM_Cubic>,<at 0.3602 0.4323 RCIM_Cubic>]}
       RedCurve {keys [<at 0.1022 1 RCIM_Cubic>,<at 0.3602 0.4323 RCIM_Cubic>]}
      }})
    Use R1 Painted Overrides=true
    Use R2 Painted Overrides=true
   ) {enabled=false, module=VC_PaintedVariantsOverride001}
   &ScaleSpriteSize(
    Uniform Curve Scale=&RandomRangeFloat(Minimum=0.8, Maximum=0.9)
    Uniform Curve Sprite Scale=DataInterface() {propertyValues {Curve {keys [
        {arriveTangent=0.8007, interpMode=RCIM_Cubic, leaveTangent=0.8007, tangentMode=RCTM_User, time=0, value=0.7}
        {arriveTangent=-0.0118, interpMode=RCIM_Cubic, leaveTangent=-0.0118, tangentMode=RCTM_User, time=1, value=1}
       ]}}}
   )
   &Drag(
    Drag=&FloatFromCurve(
     FloatCurve=DataInterface() {propertyValues {Curve {
        defaultValue=1
        keys [
         {arriveTangent=0.0291, interpMode=RCIM_Cubic, leaveTangent=0.003, tangentMode=RCTM_Break, time=0, value=0}
         {arriveTangent=4.1537, interpMode=RCIM_Cubic, leaveTangent=4.1537, tangentMode=RCTM_User, time=1, value=2}
        ]
       }}}
     Scale Curve=1.5
    )
    Use Linear Drag=true
    Use Rotational Drag=true
   )
   &ScaleColor(
    Scale Alpha=&FloatFromCurve(
     FloatCurve=DataInterface() {propertyValues {Curve {keys [<at 0 1 RCIM_Cubic>, {arriveTangent=-2.4858, interpMode=RCIM_Cubic, leaveTangent=0.085, tangentMode=RCTM_Break, time=0.6, value=0.1}]}}}
     Scale Curve=0.04
    )
    Scale RGB=&Multiply_VectorByFloat(
     Vector=<0.3 3 0>
     Float=&FloatFromCurve(
      FloatCurve=DataInterface() {propertyValues {Curve {keys [
          <at 0 1 RCIM_Cubic>
          {arriveTangentWeight=0.7768, interpMode=RCIM_Cubic, tangentMode=RCTM_Break, tangentWeightMode=RCTWM_WeightedBoth, time=0.95, value=0.03}
         ]}}}
      Scale Curve=&RandomRangeFloat(Minimum=1, Maximum=0.5)
     )
    )
    ScaleA=true
    ScaleRGB=true
   ) {module=ScaleColor001}
   &ScaleVelocity(
    Coordinate Space:ENiagaraCoordinateSpace=::World
    Velocity Scale=&MakeVector(
     X=1
     Y=1
     Z=&FloatFromCurve(FloatCurve=DataInterface() {propertyValues {Curve {keys [<at 0 0 RCIM_Cubic>,<at 1 1 RCIM_Cubic>]}}}, Scale Curve=2)
    )
   )
   &Collision(
    Age Colliding Particles=false
    KillOnCollision=true
    Module.Control Rest State Via Surface Normal=false
    Module.EnableMaxCollisionCount=false
    Module.Randomize Collision Normal Vector=false
   )
   /Niagara/Modules/Events/GenerateCollisionEvent(Module.EventDelay=true, Module.Use Velocity Threshold=true) {enabled=false}
   &InheritSourceMovement(
    Applied Position Scale=&Multiply_VectorByFloat(
     Vector=<1 1 1>
     Float=&FloatFromCurve(FloatCurve=DataInterface() {propertyValues {Curve {
         defaultValue=1
         keys [
          {arriveTangent=-1.7885, interpMode=RCIM_Cubic, leaveTangent=-1.7885, tangentMode=RCTM_User, time=0.05, value=1}
          {interpMode=RCIM_Cubic, tangentMode=RCTM_User, time=1, value=0}
         ]
        }}})
    )
    ApplyForce=false
    ApplyPosition=true
    ApplyPositionFalloff=false
   ) {module=InheritSourceMovement001}
   &DynamicMaterialParameters(
    Index 0 Param 1=1
    Index 0 Param 2=1
    Index 0 Param 3=1
    Index 1 Param 1=1
    Index 1 Param 2=1
    Index 1 Param 3=&FloatFromCurve(
     FloatCurve=DataInterface() {propertyValues {Curve {keys [
         {
          arriveTangent=-0.1398
          interpMode=RCIM_Cubic
          leaveTangent=-0.366
          leaveTangentWeight=0.146
          tangentMode=RCTM_Break
          tangentWeightMode=RCTWM_WeightedBoth
          time=0
          value=0.16
         }
         {arriveTangentWeight=0.9374, interpMode=RCIM_Cubic, tangentMode=RCTM_Break, tangentWeightMode=RCTWM_WeightedBoth, time=1, value=0.05}
        ]}}}
     Scale Curve=30
    )
    Index 1 Param 4=1
    Index 2 Param 1=1
    Index 2 Param 3=1
    Index 3 Param 3=0.5
    Index 3 Param 4=1
    Module.Index0Float0Write=true
    Module.Index0Float1Write=true
    Module.Index0Float2Write=true
    Module.Index0Float3Write=true
    Module.Index1Float0Write=true
    Module.Index1Float1Write=true
    Module.Index1Float2Write=true
    Module.Index1Float3Write=true
    Module.Index2Float0Write=true
    Module.Index2Float1Write=true
    Module.Index2Float2Write=true
    Module.Index2Float3Write=true
    Module.Index3Float0Write=true
    Module.Index3Float1Write=true
    Module.Index3Float2Write=true
    Module.Index3Float3Write=true
    Write Parameter Index 1=::NewEnumerator1
    Write Parameter Index 2=::NewEnumerator1
    Write Parameter Index 3=::NewEnumerator1
   ) {enabled=false}
   &SolveForcesAndVelocity(Clamp Velocity=false, Limit Acceleration=false)
  ]
  rendererClasses [&NiagaraSpriteRendererProperties]
  renderers [{rendererClass=&NiagaraSpriteRendererProperties}]
 }
}
summary {
 systemName=NS_Boost_LavaPsplus
 emitters [
  {emitterName=NE_VC_Boost_Ribbon_Parent, rendererClasses [&NiagaraRibbonRendererProperties]}
  {emitterName=NE_VC_Boost_SPLASH001, rendererClasses [&NiagaraSpriteRendererProperties]}
  {emitterName=NE_CS_MainTrail002, rendererClasses [&NiagaraSpriteRendererProperties]}
  {emitterName=NE_VC_GLow, rendererClasses [&NiagaraSpriteRendererProperties]}
 ]
 userVariables [
  User.AddMultiply {type=NiagaraFloat, defaultValue=0, description=""}
  User.bIsFrontendPreview {type=NiagaraBool}
  User.bIsPainted {type=NiagaraBool}
  User.blsIsCoroColor {type=NiagaraBool}
  User.bPreviewVehicleSpeed {type=NiagaraBool}
  User.Brightness {type=NiagaraFloat, defaultValue=0, description=""}
  User.ColorVariantID {type=NiagaraInt32}
  User.DTCarOffset {type=NiagaraPosition, description="", defaultValue {}}
  User.DTCarOffset {type=Vector3f, description="", defaultValue {struct=Vector3f, value=<0 0 0>}}
  User.DTCarOffset001 {type=Vector3f, description="", defaultValue {struct=Vector3f, value=<0 0 0>}}
  User.GameModeID {type=NiagaraInt32, defaultValue=3}
  User.GlowOverride {type=NiagaraFloat, defaultValue=4, description=""}
  User.PaintedColor {type=LinearColor, description="", defaultValue {struct=LinearColor, value=<rgba 1 1 1 1>}}
  User.PaintedScalar {type=NiagaraFloat, defaultValue=1, description=""}
  User.PortCount {type=NiagaraInt32, defaultValue=2}
  User.RingsIntensityB {type=NiagaraFloat, defaultValue=5, description=""}
  User.RingsIntensityG {type=NiagaraFloat, defaultValue=5, description=""}
  User.RingsIntensityR {type=NiagaraFloat, defaultValue=5, description=""}
  User.RingsOpacity {type=NiagaraFloat, defaultValue=1, description=""}
  User.SolidMask {type=NiagaraFloat, defaultValue=0, description=""}
 ]
}
